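# Material function for the PastelGlazeGrain backpack that makes its emissive color react to audio, with a parameter for enabling reactivity.
# Audio amplitude, shaped by an exponent, feeds an amplitude mask and the radius of a local-space sphere mask that boosts emissive saturation in HSV space.
# Time-of-day intensity and a camera-facing normal term shape the glow, which blends with the original emissive and replaces it in the incoming material attributes.
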MaterialFunction /BRCosmetics/Accessories/FORT_Backpacks/Backpack_F_MED_PastelGlazeGrain/Materials/MF_PastelGlazeGrain_Backpack
graph {nodes [
  n1 = GetMaterialAttributes(None=FunctionInput() {inputType=FunctionInput_MaterialAttributes}) {attributeGetTypes [#1]}
  n2 = MaterialFunctionCall(Float3=RgbToHsv!(None=@n1.EmissiveColor)) {materialFunction=&BreakOutFloat3Components}
  n3 = MaterialFunctionCall() {materialFunction=/Game/Packages/Fortress_SharedMaterials/Base_Material_Functions/BaseCharacterShaderFunctions/Audio/MF_AudioReactivity}
  n4 = NamedRerouteUsage()
  n5 = NamedRerouteDeclaration(None=Power(Base=Multiply(A=@n3.Amplitude, B=@n4), Exp=ScalarParameter() {defaultValue=3, group=PastelGlazeGrain, parameterName=Audio_Amplitude_Exp}) {constExponent=4}) {name=OverallAmp, nodeColor=<rgba 1 0 0.9609 1>}
  n6 = Constant4Vector() {constant=<rgba 0 1 0 80>}
  n7 = Multiply(A=SphereMask(
    A=LocalPosition().XYZ
    B=Constant3Vector()
    Radius=MaterialFunctionCall(Input=@n5, Input Low=@n6.R, Input High=@n6.R, Target Low=@n6.R, Target High=@n6.R) {materialFunction=&RemapValueRange}.Result
   ) {hardnessPercent=1}) {constB=0.1}
  n8 = HsvToRgb!(None=MaterialFunctionCall(X=@n2.R, Y=Add(A=@n2.G, B=@n7), Z=@n2.R) {materialFunction=&MakeFloat3}.Result)
  n9 = VectorParameter() {group=PastelGlazeGrain, parameterName=EmissiveIntensity_TOD, defaultValue=<rgba 2 6 2 2>}
  n10 = Add(A=LinearInterpolate(
    B=MaterialFunctionCall(Value_Morning=@n9.R, Value_Day=@n9.R, Value_Evening=@n9.R, Value_Night=@n9.R) {materialFunction=&Scalar_Time_of_Day}.Result
    Alpha=NamedRerouteUsage()
   ) {constA=-0.75, constB=2})
  n11 = DotProduct(A=CameraVectorWS(), B=VertexNormalWS() {desc=Model Space Normal})
  n12 = ConstantBiasScale(None=@n11)
  n13 = LinearInterpolate(A=Multiply(A=Multiply(A=@n8, B=@n10), B=Power(Base=@n12) {constExponent=4}), B=@n1.EmissiveColor, Alpha=NamedRerouteUsage())
  n14 = NamedRerouteUsage()
  FunctionOutput(None=SetMaterialAttributes!(MaterialAttributes=@n1.MaterialAttributes, EmissiveColor=LinearInterpolate(A=@n1.EmissiveColor, B=@n13, Alpha=@n14)) {attributeSetTypes [#1]}) {outputName=OUT}
  NamedRerouteDeclaration(None=Step(Y=@n5) {constX=0.1, constY=0.1}) {name=AMPMask, nodeColor=<rgba 1 0 0.1641 1>}
  NamedRerouteDeclaration(None=ScalarParameter() {defaultValue=1, group=PastelGlazeGrain, parameterName=EnableAudioReactivity}) {name=ReactivityOn, nodeColor=<rgba 0.1719 0 1 1>}
  OneMinus(None=@n11)
 ]}
