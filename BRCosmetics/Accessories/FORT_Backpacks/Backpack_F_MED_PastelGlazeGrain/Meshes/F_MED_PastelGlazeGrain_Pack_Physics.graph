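# Physics asset for the PastelGlazeGrain backpack cosmetic, giving its pack and mirrored left and right wings capsule bodies for physics.
# Each wing master joint hangs off the pack and carries upper base, mid base, lower feather and ponytail branches.
# Most small inner and outer wing joints lock swing and twist, leaving the upper-outer pair unlocked.
# Ponytails chain through joints 01, 02, 03 and 05, with twist locked on 02 and 03.
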PhysicsAsset /BRCosmetics/Accessories/FORT_Backpacks/Backpack_F_MED_PastelGlazeGrain/Meshes/F_MED_PastelGlazeGrain_Pack_Physics
bodies {
 pack [{shapeType=Capsule, length=5.4376, radius=4.5624, center=<0 5.8304 0>, rotation=<rot -0.0002 -90 -90>}]
 wing_l_lower_inner_01_jnt [{shapeType=Capsule, length=3, radius=3, shapeName=wing_l_lower_inner_01_jnt_capsule, rotation=<rot 90 0 0>}]
 wing_l_lower_out_01_jnt [{shapeType=Capsule, length=3, radius=3, shapeName=wing_l_lower_out_01_jnt_capsule, rotation=<rot 90 0 0>}]
 wing_l_master_jnt [{shapeType=Capsule, length=11.6054, radius=4.9835, center=<8.1799 -4.5419 -3.0226>, rotation=<rot 0 -103.9935 -53.6803>}]
 wing_l_mid_base_jnt [{shapeType=Capsule, length=12.5505, radius=6.8379, center=<11.7775 0.1327 -3.7392>, rotation=<rot 90 0 0>}]
 wing_l_mid_inner_01_jnt [{shapeType=Capsule, length=3, radius=3, shapeName=wing_l_mid_inner_01_jnt_capsule, rotation=<rot 90 0 0>}]
 wing_l_mid_out_01_jnt [{shapeType=Capsule, length=3, radius=3, shapeName=wing_l_mid_out_01_jnt_capsule, rotation=<rot 90 0 0>}]
 wing_l_mid_out_02_jnt [{shapeType=Capsule, length=3, radius=3, shapeName=wing_l_mid_out_02_jnt_capsule, rotation=<rot 90 0 0>}]
 wing_l_mid_out_03_jnt [{shapeType=Capsule, length=3, radius=3, shapeName=wing_l_mid_out_03_jnt_capsule, rotation=<rot 90 0 0>}]
 wing_l_ponytail_01_jnt [{shapeType=Capsule, length=12.5032, radius=4.0584, center=<13.7583 3.1459 4.1971>, rotation=<rot 0 106.6019 -79.8779>}]
 wing_l_ponytail_02_jnt [{
   shapeType=Capsule
   length=2.1604
   radius=0.7746
   shapeName=wing_l_ponytail_02_jnt_capsule
   center=<2.7748 -0.3953 0.091>
   rotation=<rot 0 93.6264 -89.5756>
  }]
 wing_l_ponytail_03_jnt [{
   shapeType=Capsule
   length=3.5706
   radius=0.8603
   shapeName=wing_l_ponytail_03_jnt_capsule
   center=<3.4369 -0.139 -0.1012>
   rotation=<rot 0 -81.9773 -86.3799>
  }]
 wing_l_ponytail_05_jnt [{shapeType=Capsule, length=3.1136, radius=2.4791, shapeName=wing_l_ponytail_05_jnt_capsule}]
 wing_l_up_inner_jnt [{shapeType=Capsule, length=3, radius=3, shapeName=wing_l_up_inner_jnt_capsule, rotation=<rot 90 0 0>}]
 wing_l_up_out_01_jnt [{shapeType=Capsule, length=3, radius=3, shapeName=wing_l_up_out_01_jnt_capsule, rotation=<rot 90 0 0>}]
 wing_l_up_out_02_jnt [{shapeType=Capsule, length=3, radius=3, shapeName=wing_l_up_out_02_jnt_capsule, rotation=<rot 90 0 0>}]
 wing_l_upper_base_jnt [{shapeType=Capsule, length=16.8153, radius=7.1742, center=<15.933 0.6618 -2.8428>, rotation=<rot 0 -79.6982 -79.3219>}]
 wing_r_lower_inner_01_jnt [{shapeType=Capsule, length=3, radius=3, shapeName=wing_r_lower_inner_01_jnt_capsule, rotation=<rot 90 0 0>}]
 wing_r_lower_out_01_jnt [{shapeType=Capsule, length=3, radius=3, shapeName=wing_r_lower_out_01_jnt_capsule, rotation=<rot 90 0 0>}]
 wing_r_master_jnt [{shapeType=Capsule, length=11.6054, radius=4.9835, center=<-8.1799 4.5419 3.0226>, rotation=<rot 0 -103.9935 -53.6803>}]
 wing_r_mid_base_jnt [{shapeType=Capsule, length=12.5505, radius=6.8379, center=<-11.7775 -0.1327 3.7392>, rotation=<rot 90 0 0>}]
 wing_r_mid_inner_01_jnt [{shapeType=Capsule, length=3, radius=3, shapeName=wing_r_mid_inner_01_jnt_capsule, rotation=<rot 90 0 0>}]
 wing_r_mid_out_01_jnt [{shapeType=Capsule, length=3, radius=3, shapeName=wing_r_mid_out_01_jnt_capsule, rotation=<rot 90 0 0>}]
 wing_r_mid_out_02_jnt [{shapeType=Capsule, length=3, radius=3, shapeName=wing_r_mid_out_02_jnt_capsule, rotation=<rot 90 0 0>}]
 wing_r_mid_out_03_jnt [{shapeType=Capsule, length=3, radius=3, shapeName=wing_r_mid_out_03_jnt_capsule, rotation=<rot 90 0 0>}]
 wing_r_ponytail_01_jnt [{shapeType=Capsule, length=12.5032, radius=4.0584, center=<-13.7583 -3.1459 -4.1971>, rotation=<rot 0 106.6019 -79.8779>}]
 wing_r_ponytail_02_jnt [{
   shapeType=Capsule
   length=2.1604
   radius=0.7746
   shapeName=wing_r_ponytail_02_jnt_capsule
   center=<-2.7748 0.3953 -0.091>
   rotation=<rot 0 93.6264 -89.5756>
  }]
 wing_r_ponytail_03_jnt [{
   shapeType=Capsule
   length=2.104
   radius=0.6905
   shapeName=wing_r_ponytail_03_jnt_capsule
   center=<-2.0735 0.4097 -0.1115>
   rotation=<rot 0 90.4938 -87.5044>
  }]
 wing_r_ponytail_05_jnt [{shapeType=Capsule, length=3.1136, radius=2.4791, shapeName=wing_l_ponytail_05_jnt_capsule, rotation=<rot 0 0 -180>}]
 wing_r_up_inner_jnt [{shapeType=Capsule, length=3, radius=3, shapeName=wing_r_up_inner_jnt_capsule, rotation=<rot 90 0 0>}]
 wing_r_up_out_01_jnt [{shapeType=Capsule, length=3, radius=3, shapeName=wing_r_up_out_01_jnt_capsule, rotation=<rot 90 0 0>}]
 wing_r_up_out_02_jnt [{shapeType=Capsule, length=3, radius=3, shapeName=wing_r_up_out_02_jnt_capsule, rotation=<rot 90 0 0>}]
 wing_r_upper_base_jnt [{shapeType=Capsule, length=16.8153, radius=7.1742, center=<-15.933 -0.6618 2.8428>, rotation=<rot 0 -79.6982 -79.3219>}]
}
constraints [
 {bone1Name=wing_l_master_jnt, bone2Name=pack, swing1LimitDegrees=3, swing2LimitDegrees=7, twistLimitDegrees=3}
 {bone1Name=wing_l_upper_base_jnt, bone2Name=wing_l_master_jnt, swing1LimitDegrees=10, swing2LimitDegrees=20, twistLimitDegrees=5}
 {bone1Name=wing_l_mid_base_jnt, bone2Name=wing_l_master_jnt, swing1LimitDegrees=10, swing2LimitDegrees=20, twistLimitDegrees=5}
 {bone1Name=wing_l_ponytail_01_jnt, bone2Name=wing_l_master_jnt, swing1LimitDegrees=10, swing2LimitDegrees=20, twistLimitDegrees=5}
 {bone1Name=wing_r_master_jnt, bone2Name=pack, swing1LimitDegrees=5, swing2LimitDegrees=7, twistLimitDegrees=3}
 {bone1Name=wing_r_upper_base_jnt, bone2Name=wing_r_master_jnt, swing1LimitDegrees=10, swing2LimitDegrees=20, twistLimitDegrees=5}
 {bone1Name=wing_r_mid_base_jnt, bone2Name=wing_r_master_jnt, swing1LimitDegrees=10, swing2LimitDegrees=20, twistLimitDegrees=5}
 {bone1Name=wing_r_ponytail_01_jnt, bone2Name=wing_r_master_jnt, swing1LimitDegrees=10, swing2LimitDegrees=20, twistLimitDegrees=5}
 {
  bone1Name=wing_l_lower_inner_01_jnt
  bone2Name=wing_l_master_jnt
  swing1LimitDegrees=15
  swing1Motion=Locked
  swing2LimitDegrees=25
  swing2Motion=Locked
  twistLimitDegrees=5
  twistMotion=Locked
 }
 {
  bone1Name=wing_l_lower_out_01_jnt
  bone2Name=wing_l_master_jnt
  swing1LimitDegrees=5
  swing1Motion=Locked
  swing2LimitDegrees=5
  swing2Motion=Locked
  twistLimitDegrees=5
  twistMotion=Locked
 }
 {
  bone1Name=wing_l_mid_inner_01_jnt
  bone2Name=wing_l_mid_base_jnt
  swing1LimitDegrees=25
  swing1Motion=Locked
  swing2LimitDegrees=25
  swing2Motion=Locked
  twistLimitDegrees=5
  twistMotion=Locked
 }
 {
  bone1Name=wing_l_mid_out_01_jnt
  bone2Name=wing_l_mid_base_jnt
  swing1LimitDegrees=25
  swing1Motion=Locked
  swing2LimitDegrees=15
  swing2Motion=Locked
  twistLimitDegrees=5
  twistMotion=Locked
 }
 {
  bone1Name=wing_l_mid_out_02_jnt
  bone2Name=wing_l_mid_base_jnt
  swing1LimitDegrees=5
  swing1Motion=Locked
  swing2LimitDegrees=5
  swing2Motion=Locked
  twistLimitDegrees=5
  twistMotion=Locked
 }
 {
  bone1Name=wing_l_mid_out_03_jnt
  bone2Name=wing_l_mid_base_jnt
  swing1LimitDegrees=15
  swing1Motion=Locked
  swing2LimitDegrees=35
  swing2Motion=Locked
  twistLimitDegrees=5
  twistMotion=Locked
 }
 {
  bone1Name=wing_l_up_inner_jnt
  bone2Name=wing_l_upper_base_jnt
  swing1LimitDegrees=5
  swing1Motion=Locked
  swing2LimitDegrees=5
  swing2Motion=Locked
  twistLimitDegrees=5
  twistMotion=Locked
 }
 {bone1Name=wing_l_up_out_01_jnt, bone2Name=wing_l_upper_base_jnt, swing1LimitDegrees=25, swing2LimitDegrees=15, twistLimitDegrees=10}
 {bone1Name=wing_l_up_out_02_jnt, bone2Name=wing_l_upper_base_jnt, swing1LimitDegrees=35, swing2LimitDegrees=15, twistLimitDegrees=15}
 {
  bone1Name=wing_r_lower_inner_01_jnt
  bone2Name=wing_r_master_jnt
  swing1LimitDegrees=15
  swing1Motion=Locked
  swing2LimitDegrees=25
  swing2Motion=Locked
  twistLimitDegrees=5
  twistMotion=Locked
 }
 {
  bone1Name=wing_r_lower_out_01_jnt
  bone2Name=wing_r_master_jnt
  swing1LimitDegrees=5
  swing1Motion=Locked
  swing2LimitDegrees=5
  swing2Motion=Locked
  twistLimitDegrees=5
  twistMotion=Locked
 }
 {
  bone1Name=wing_r_mid_inner_01_jnt
  bone2Name=wing_r_mid_base_jnt
  swing1LimitDegrees=25
  swing1Motion=Locked
  swing2LimitDegrees=25
  swing2Motion=Locked
  twistLimitDegrees=5
  twistMotion=Locked
 }
 {
  bone1Name=wing_r_mid_out_01_jnt
  bone2Name=wing_r_mid_base_jnt
  swing1LimitDegrees=25
  swing1Motion=Locked
  swing2LimitDegrees=15
  swing2Motion=Locked
  twistLimitDegrees=5
  twistMotion=Locked
 }
 {
  bone1Name=wing_r_mid_out_02_jnt
  bone2Name=wing_r_mid_base_jnt
  swing1LimitDegrees=5
  swing1Motion=Locked
  swing2LimitDegrees=5
  swing2Motion=Locked
  twistLimitDegrees=5
  twistMotion=Locked
 }
 {
  bone1Name=wing_r_mid_out_03_jnt
  bone2Name=wing_r_mid_base_jnt
  swing1LimitDegrees=15
  swing1Motion=Locked
  swing2LimitDegrees=35
  swing2Motion=Locked
  twistLimitDegrees=5
  twistMotion=Locked
 }
 {
  bone1Name=wing_r_up_inner_jnt
  bone2Name=wing_r_upper_base_jnt
  swing1LimitDegrees=5
  swing1Motion=Locked
  swing2LimitDegrees=5
  swing2Motion=Locked
  twistLimitDegrees=5
  twistMotion=Locked
 }
 {bone1Name=wing_r_up_out_01_jnt, bone2Name=wing_r_upper_base_jnt, swing1LimitDegrees=25, swing2LimitDegrees=15, twistLimitDegrees=10}
 {bone1Name=wing_r_up_out_02_jnt, bone2Name=wing_r_upper_base_jnt, swing1LimitDegrees=35, swing2LimitDegrees=15, twistLimitDegrees=15}
 {
  bone1Name=wing_l_ponytail_02_jnt
  bone2Name=wing_l_ponytail_01_jnt
  swing1LimitDegrees=13
  swing2LimitDegrees=2
  twistLimitDegrees=45
  twistMotion=Locked
 }
 {
  bone1Name=wing_l_ponytail_03_jnt
  bone2Name=wing_l_ponytail_02_jnt
  swing1LimitDegrees=25
  swing2LimitDegrees=15
  twistLimitDegrees=45
  twistMotion=Locked
 }
 {
  bone1Name=wing_r_ponytail_02_jnt
  bone2Name=wing_r_ponytail_01_jnt
  swing1LimitDegrees=13
  swing2LimitDegrees=2
  twistLimitDegrees=45
  twistMotion=Locked
 }
 {
  bone1Name=wing_r_ponytail_03_jnt
  bone2Name=wing_r_ponytail_02_jnt
  swing1LimitDegrees=25
  swing2LimitDegrees=15
  twistLimitDegrees=45
  twistMotion=Locked
 }
 {bone1Name=wing_l_ponytail_05_jnt, bone2Name=wing_l_ponytail_03_jnt, swing1LimitDegrees=25, swing2LimitDegrees=25, twistLimitDegrees=5}
 {bone1Name=wing_r_ponytail_05_jnt, bone2Name=wing_r_ponytail_03_jnt, swing1LimitDegrees=25, swing2LimitDegrees=25, twistLimitDegrees=5}
]
